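# Material function that fakes a runtime virtual texture by sampling static world-mapped textures for base color, specular, roughness, normal and height.
# UVs come from world XY centered over a world-size parameter, unless the UVs input differs from its -3.14 sentinel in X.
# Height remaps a separate texture to a fixed range plus an offset parameter.
# Mask outputs constant 1, while the MipLevel input and SRH height channel feed no output.
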MaterialFunction /Game/ArtTools/Tech_Art/FakeRVT/MF_FakeRVT
graph {nodes [
  FunctionInput() {bUsePreviewValueAsDefault=true, inputName=MipLevel, inputType=FunctionInput_Scalar, sortPriority=50, previewValue=<1 0 0 1>}
  n1 = FunctionInput() {bUsePreviewValueAsDefault=true, inputName=UVs, inputType=FunctionInput_Vector2, sortPriority=5, previewValue=<-3.14 -3.14 0 1>}
  n2 = MaterialFunctionCall(Float2=@n1) {materialFunction=&BreakOutFloat2Components}
  n3 = ComponentMask(None=FunctionInput(Preview=WorldPosition().XYZ) {bUsePreviewValueAsDefault=true, inputName=WorldPosition, sortPriority=10}) {g=true, r=true}
  n4 = ScalarParameter() {defaultValue=260096, group=FakeRVT, parameterName=FRVT_WorldSize}
  n5 = Divide(A=@n4)
  n6 = If(A=Add(A=@n2.R, B=@n2.R), "A > B"=@n1, "A == B"=Divide(A=Add(A=@n3, B=@n5), B=@n4), "A < B"=@n1) {constB=-6.28}
  FunctionOutput(None=TextureSampleParameter2D(UVs=@n6) {group=FakeRVT, parameterName=FakeRVT_BaseColor, texture=/Game/ArtTools/Tech_Art/FakeRVT/Textures/FRVT_BaseColor}.RGB) {outputName=BaseColor, sortPriority=56}
  n7 = TextureSampleParameter2D(UVs=@n6) {
   group=FakeRVT
   parameterName=FakeRVT_SRH
   samplerType=SAMPLERTYPE_Masks
   texture=/Game/ArtTools/Tech_Art/FakeRVT/Textures/FRVT_SRH
   channelNames {a="", b=Height, g=Roughness, r=Specular}
  }
  FunctionOutput(None=@n7.Specular) {outputName=Specular, sortPriority=100}
  FunctionOutput(None=@n7.Roughness) {outputName=Roughness, sortPriority=123}
  n8 = TextureSampleParameter2D(UVs=@n6) {group=FakeRVT, parameterName=FakeRVT_Normal, samplerType=SAMPLERTYPE_Normal, texture=/Game/ArtTools/Tech_Art/FakeRVT/Textures/FRVT_Normal}
  FunctionOutput(None=@n8.RGB) {outputName=Normal, sortPriority=231}
  n9 = TextureSampleParameter2D(UVs=@n6) {
   group=FakeRVT
   parameterName=FakeRVT_Height
   samplerType=SAMPLERTYPE_LinearColor
   texture=/Game/ArtTools/Tech_Art/FakeRVT/Textures/FRVT_Height
  }
  FunctionOutput(None=Add(A=LinearInterpolate(Alpha=@n9.R) {constA=-3100, constB=15000}, B=ScalarParameter() {parameterName=Param_1})) {outputName=Height, sortPriority=321}
  FunctionOutput(None=Constant() {r=1}) {outputName=Mask, sortPriority=3425}
  Multiply(A=@n7.Height) {constB=15000}
  ScalarParameter() {defaultValue=260096}
 ]}
