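# Test material for the fake runtime virtual texture system that blends a surface into the data sampled by MF_FakeRVT.
# The blend mask compares sampled height to biased world Z, combining the difference's sign with a tunable SphereMask falloff.
# The mask lerps base color, specular and roughness toward the sampled values, from a pink parameter color for base color.
# The world-space normal lerps oppositely, from sampled normal to transformed Z-up vector.
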Material /Game/ArtTools/Tech_Art/FakeRVT/M_FakeRVT_TestMaterial
graph {
 properties {bTangentSpaceNormal=false, pixelDepthOffsetMode=PDOM_Legacy, usedShadingModels=Default Lit}
 nodes [
  n1 = MaterialFunctionCall!(UVs, WorldPosition, MipLevel) {materialFunction=&/MF_FakeRVT}
  n2 = Add(A=WorldPosition().Z, B=ScalarParameter() {defaultValue=3, parameterName=Bias})
  n3 = Subtract(A=@n1.Height, B=@n2)
  n4 = SphereMask(
   A=@n1.Height
   B=@n2
   Radius=ScalarParameter() {defaultValue=340.0443, parameterName=Blend height, sortPriority=63}
   Hardness=ScalarParameter() {defaultValue=0.384, parameterName=Hardness}
  )
  n5 = ScalarParameter() {defaultValue=1}
  n6 = Min(A=Max(A=Sign(None=@n3), B=Power(Base=@n4, Exp=@n5)), B=ScalarParameter() {defaultValue=1, parameterName=Param_1})
  LinearInterpolate(A=VectorParameter() {parameterName=vvv, defaultValue=<rgba 1 0.1101 0.8762 1>}.RGB, B=@n1.BaseColor, Alpha=@n6)
  LinearInterpolate(B=@n1.Specular, Alpha=@n6)
  LinearInterpolate(B=@n1.Roughness, Alpha=@n6)
  LinearInterpolate(A=@n1.Normal, B=Transform(None=Constant3Vector() {constant=<rgba 0 0 1 0>}), Alpha=@n6)
 ]
}
